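; Heatmap manager actor that visualizes a data object's locations by spawning one element actor per entry.
; Setting a data object runs CleanUp then DrawHeatmap, which spawns elements from a soft class, tracks them and calls DrawElement on each.
; Spawn and destroy go as messages to itself when a player controller exists, otherwise to an editor component, supporting both game and editor use.
; Destroying the manager also cleans up its elements.
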(event EventDestroyed
  (HeatmapRenderer|CleanUp))

(fn CleanUp
  (bind _target (Protected|GetEditororGameTarget))
  (HeatmapRenderer|DestroyElements(Message) _target (Variables|Protected|GetHeatmapElements)))

(fn destroyElements
  (params (elements :type Array%20of%20BP%20Heatmap%20Renderer%20Element%20Object%20References))
  (bind (_arrayindex _arrayelement) (Utilities|Array|ReverseforEachLoop elements)
    (:LoopBody
      (Actor|DestroyActor _arrayelement))))

(fn DrawHeatmap
  (for (_ _array_index) (Variables|/Game/Blueprints/HeatmapRenderer/Data/HeatmapRendererDataObject.HeatmapRendererDataObject_C|GetLocations (Variables|HeatmapRenderer|GetCurrentDataObject))
    (bind _target (Protected|GetEditororGameTarget))
    (bind _newelement (HeatmapRenderer|SpawnElement(Message) _target (Variables|HeatmapRenderer|GetElementClass) (Variables|HeatmapRenderer|GetCurrentDataObject) _array_index))
    (Protected|HandleNewElementAdded _newelement)))

(fn GetEditorOrGameTarget
  (Utilities|IsValid (Game|GetPlayerController 0)
    (:Is%20Valid
      (return self))
    (:Is%20Not%20Valid
      (return (Variables|Default|GetHeatmapRenderer_EditorComponent)))))

(fn HandleDataObjectChanged
  (Utilities|FlowControl|Sequence
    (:then_0
      (HeatmapRenderer|CleanUp))
    (:then_1
      (HeatmapRenderer|DrawHeatmap))))

(fn HandleNewElementAdded
  (params (element :type BP%20Heatmap%20Renderer%20Element%20Object%20Reference))
  (bind _output_get (LocalVariables|Setl_element element))
  (Utilities|Array|Add (Variables|Protected|GetHeatmapElements) _output_get)
  (Class|BPHeatmapRendererElement|DrawElement (LocalVariables|Getl_element)))

(fn SetHeatmapDataObject
  (params (data :type Heatmap%20Renderer%20Data%20Object%20Object%20Reference))
  (Variables|HeatmapRenderer|SetCurrentDataObject data)
  (HeatmapRenderer|HandleDataObjectChanged))

(fn spawnElement
  (params (class :type BP%20Heatmap%20Renderer%20Element%20Soft%20Class%20Reference) (data :type Heatmap%20Renderer%20Data%20Object%20Object%20Reference%20%28by%20ref%29) (dataIndex :type Integer))
  (bind _location (Class|HeatmapRendererDataObject|GetBasicDataOutput data dataIndex))
  (bind _returnvalue (Game|SpawnActorfromClass (Utilities|ResolveSoftReference class) (Math|Conversions|ToTransform(Vector) _location) "Undefined" "MultiplyWithRoot" self data dataIndex self 0))
  (return _returnvalue))

(fn ConstructionScript)
