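; BP_HeatmapRenderer_Manager_HeightMap is the height-map variant of a heatmap renderer manager.
; Its DrawHeatmap function calls DrawMesh on the HeatmapRenderer_HeightMapComponent, referencing the I_HeatmapRenderer_EditorBridge interface from the renderer's Editor folder.
; BeginPlay, overlap and Tick events are empty, and the construction script only calls its parent.
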(event EventBeginPlay)

(event Collision|EventActorBeginOverlap
  (params (OtherActor :type Actor%20Object%20Reference)))

(event EventTick
  (params (DeltaSeconds :type Float%20%28single-precision%29)))

(fn DrawHeatmap
  (HeatmapRenderer||Mesh|DrawMesh (Variables|Default|GetHeatmapRenderer_HeightMapComponent) :!class /Game/Blueprints/HeatmapRenderer/Editor/I_HeatmapRenderer_EditorBridge.I_HeatmapRenderer_EditorBridge_C))

(fn ConstructionScript
  (|CallParentFunction))
