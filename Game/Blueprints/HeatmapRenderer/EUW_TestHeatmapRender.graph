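# Editor utility widget that builds a transient heatmap renderer actor from actor locations in the level, with a details view for its settings.
# GatherData reads World Partition actor descriptors, keeps bounds centres of non-script classes whose NativeParentClass registry tag contains ClassSubstring, and bins them into grid cells.
# Cell keys and values fill a data object passed to a newly spawned manager, after old managers are removed.
# A seeded random test-location generator exists but is unwired.
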EditorUtilityWidgetBlueprint /Game/Blueprints/HeatmapRenderer/EUW_TestHeatmapRender
graphs [
 CleanUpOldManagers {nodes [
   n1 = GetAllLevelActors(self=Get EditorActorSubsystem())
   > n2 = Filter by Class!(
    execute
    TargetArray:Array of Object References=@n1
    ObjectClass:Object Class Reference=&/BP_HeatmapRenderer_Manager.BP_HeatmapRenderer_Manager_C
    FilterType:EEditorScriptingFilterType Enum
   ) -> (then, ReturnValue:Array of Heatmap Renderer Manager Object References)
   n3 = GetAllActorsOfClass!(
    execute=@n5.then
    WorldContextObject:Object Reference
    ActorClass:Actor Class Reference=&BP_HeatmapRenderer_Manager_C
   ) -> (then, OutActors:Array of Heatmap Renderer Manager Object References)
   n4 = Reverse for Each Loop!(Exec:Exec=[@n3.then, @n2.then], Array:Array of Heatmap Renderer Manager Object References=@n3) -> (LoopBody:Exec, ArrayIndex:int, ArrayElement:Heatmap Renderer Manager Object Reference, Completed:Exec) {is=title K2Node_MacroInstance}
   > CleanUp(self:Heatmap Renderer Manager Object Reference=@n4.ArrayElement)
   > DestroyActor!(execute, self:Editor Actor Subsystem Object Reference=Get EditorActorSubsystem(), ActorToDestroy:Actor Object Reference=@n4.ArrayElement) -> (then, ReturnValue:bool)
   n5 = CleanUpOldManagers() {is=title K2Node_FunctionEntry}
  ]}
 EventGraph {nodes [
   "On Clicked (TA_Generic01_Button)"!() -> (OutputDelegate, then, widget:User Widget Object Reference, event:Pointer Event Structure) {is=title K2Node_ComponentBoundEvent}
   > GatherData()
   "On Clicked (btnCleanUp)"!() -> (OutputDelegate, then, widget:User Widget Object Reference, event:Pointer Event Structure) {is=title K2Node_ComponentBoundEvent}
   > CleanUpOldManagers() {is=title K2Node_CallFunction}
   Event Construct()
   > SetObject(self=Get DetailsView_0() -> (DetailsView_0:Details View Object Reference), NewObject=Self-Reference())
  ]}
 GatherData {nodes [
   n1 = GetTestLocations!(execute, self) -> (then, Locations:Array of Vectors)
   n2 = GetActorLocations!(execute=@n8.then, self) -> (then, Array:Array of Vectors)
   n3 = SortLocationsToCells!(
    execute=[@n1.then, @n2.then]
    locations:Array of Vectors=@n2
    gridSize:double=Get GridSize() -> (GridSize:double)
    __WorldContext:Object Reference
   ) -> (then, gridCellData:"Map of Vectors to Float (double-precision)s")
   > n4 = Keys!(execute, TargetMap:"Map of Vectors to Float (double-precision)s"=@n3) -> (then, Keys:Array of Vectors)
   > n5 = Values!(execute, TargetMap:"Map of Vectors to Float (double-precision)s"=@n3) -> (then, Values:"Array of Float (double-precision)s")
   > n6 = Construct Object from Class!(
    execute
    Class=/Game/Blueprints/HeatmapRenderer/Data/HeatmapRendererDataObject.HeatmapRendererDataObject_C
    self
    Locations:Array of Vectors=@n4
    Severities:"Array of Float (double-precision)s"=@n5
   ) -> (then, ReturnValue:Heatmap Renderer Data Object Object Reference)
   > CleanUpOldManagers()
   > n7 = SpawnActorFromClass!(
    execute
    self:Editor Actor Subsystem Object Reference=Get EditorActorSubsystem()
    ActorClass:Actor Class Reference=Get ActorToSpawn() -> (ActorToSpawn:Heatmap Renderer Manager Class Reference)
    Location:Vector=MakeVector_NetQuantize!(X:double=0, Y:double=0, Z:double=Get HeightOffset() -> (HeightOffset:double)) -> (ReturnValue:Vector Net Quantize Structure)
    Rotation:Rotator
    bTransient:bool=true
   ) -> (then, ReturnValue:Heatmap Renderer Manager Object Reference)
   > SetHeatmapDataObject!(execute, self:Heatmap Renderer Manager Object Reference=@n7, data:Heatmap Renderer Data Object Object Reference=@n6) -> (then)
   n8 = GatherData() {is=title K2Node_FunctionEntry}
  ]}
 GetActorLocations {nodes [
   GetActorLocations() {is=title K2Node_FunctionEntry}
   > n1 = Sequence!(execute) -> (then_0, then_1, then_2:Exec)
   > n2 = GetActorDescs!(execute) -> (then, OutActorDescs:Array of Actor Desc Structures, ReturnValue:bool)
   > n3 = For Each Loop(Array:Array of Actor Desc Structures=@n2) -> (Array Element:Actor Desc Structure)
   n4 = Break Actor Desc!(ActorDesc:"Actor Desc Structure (by ref)"=@n3) -> (
    Guid:Guid Structure
    NativeClass:Object Class Reference
    Class:Soft Object Path Structure
    Name:Name
    Label:Name
    Bounds:Box Structure
    RuntimeGrid:Name
    bIsSpatiallyLoaded:bool=~
    bActorIsEditorOnly:bool=~
    ActorPackage:Name
    ActorPath:Name
    FolderPath:Name
    DataLayerAssets:Array of Soft Object Path Structures
   ) {is=title K2Node_BreakStruct}
   > n5 = Branch(Condition=@n4.bActorIsEditorOnly)
   Branch(execute=@n5.else, Condition=@n4.bIsSpatiallyLoaded)
   n6 = BreakSoftObjectPath(InSoftObjectPath=@n4.Class)
   > Branch(Condition=NOT Boolean(A=StartsWith(SourceString=@n6, InPrefix=/Script)))
   n7 = BreakSoftObjectPath(InSoftObjectPath=@n4.Class)
   > Add(
    TargetArray:Array of Names=Get l_packageNames() -> (l_packageNames:Array of Names)
    NewItem:"Name (by ref)"=String To Name(InString=Split(SourceString=@n7, InStr=., SearchDir=FromEnd))
   ) {is=title K2Node_CallArrayFunction}
   > Add(
    TargetArray:Array of Vectors=Get l_allLocs() -> (l_allLocs:Array of Vectors)
    NewItem:"Vector (by ref)"=GetBoxCenterSize!(Box:Box Structure=@n4.Bounds) -> (Center:Vector, Dimensions:Vector)
   )
   n8 = For Each Loop(Exec=@n1.then_1, Array:Array of Names=Get l_packageNames()) -> (Array Element:Name)
   > n9 = Branch(Condition="Not Equal Exactly (String)"(A=Get ClassSubstring() -> (ClassSubstring:String)))
   > n10 = GetAssetsByPackageName!(
    execute
    self:Asset Registry Interface=GetAssetRegistry()
    PackageName:Name=@n8
    bIncludeOnlyOnDiskAssets:bool=false
    bSkipARFilteredAssets:bool=true
   ) -> (then, OutAssetData:Array of Asset Data Structures, ReturnValue:bool)
   > Branch(Condition=Is Valid Index!(TargetArray:Array of Asset Data Structures=@n10, IndexToTest:int) -> (ReturnValue:bool))
   n11 = MakeLiteralName(Value=NativeParentClass)
   n12 = "Get (a ref)"(Array:Array of Asset Data Structures=@n10, Dimension 1=0) -> (Output:"Asset Data Structure (by ref)")
   > n13 = Branch(Condition=Contains(SearchIn=GetTagValue(InAssetData=@n12, InTagName=@n11), Substring=Get ClassSubstring()))
   Add(
    execute=[@n13.then, @n9.else]
    TargetArray=Get l_locs() -> (l_locs:Array of Vectors)
    NewItem="Get (a ref)"(Array:Array of Vectors=Get l_allLocs(), Dimension 1=@n8.Array Index) -> (Output:"Vector (by ref)")
   )
   Return Node!(execute=@n1.then_2, Array:Array of Vectors=Get l_locs())
   For Each Loop(Array:Array of Wildcards) -> (Array Element:Wildcard)
  ]}
 GetLocations {nodes [
   n1 = GetLocations!() -> (then, stream:Random Stream Structure) {is=title K2Node_FunctionEntry}
   > Set l_stream(l_stream:Random Stream Structure=@n1) -> (Output_Get:Random Stream Structure)
   > n2 = For Loop(LastIndex=50000)
   Return Node!(execute=@n2.Completed, Locations:Array of Vectors=Get l_vectors() -> (l_vectors:Array of Vectors))
   n3 = RandomFloatFromStream!(Stream:"Random Stream Structure (by ref)"=Get l_stream() -> (l_stream:Random Stream Structure)) -> (ReturnValue:float)
   n4 = RandomFloatFromStream(Stream=Get l_stream())
   n5 = RandomFloatFromStream(Stream=Get l_stream())
   > Add(
    TargetArray=Get l_vectors()
    NewItem=vector * vector(A=MakeVector(X=@n4, Y=@n5, Z=@n3), B=Get TestRadius() -> (TestRadius:double)) -> (ReturnValue=~)
   )
  ]}
 GetTestLocations {nodes [
   GetTestLocations() {is=title K2Node_FunctionEntry}
   > n1 = Set l_stream(l_stream=MakeRandomStream!(InitialSeed:int=0) -> (ReturnValue:Random Stream Structure))
   > n2 = GetLocations!(execute, self, stream:Random Stream Structure=@n1) -> (then, Locations:Array of Vectors) {is=title K2Node_CallFunction}
   > Return Node(Locations=@n2)
  ]}
]
variables [
 ActorToSpawn {type=Class, default=&/BP_HeatmapRenderer_Manager_HeightMap.BP_HeatmapRenderer_Manager_HeightMap_C}
 ClassSubstring {type=string, default=Building}
 ColorCurve {type=CurveLinearColor, default=/Game/Blueprints/HeatmapRenderer/Content/Curves/HeatmapRenderColorCurve_01}
 GridSize {type=number, default=500}
 HeightOffset {type=number, default=500}
 TestRadius {type=number, default=250000}
]
widgets {
 info {rootWidgetClass=&StackBox, widgetCount=4}
 widgets [
  StackBox_0 {widgetClassPath=&StackBox, widget=StackBox_0, properties {orientation=Orient_Vertical}}
  TA_Generic01_Button {
   widgetClassPath=&TA_Generic01_Button
   parent=StackBox_0
   widget=TA_Generic01_Button
   properties {label=Do It}
   slot {class=StackBoxSlot}
  }
  btnCleanUp {
   widgetClassPath=&TA_Generic01_Button
   parent=StackBox_0
   widget=btnCleanUp
   properties {label=Clean Up}
   slot {class=StackBoxSlot}
  }
  DetailsView_0 {
   widgetClassPath=&DetailsView
   parent=StackBox_0
   widget=DetailsView_0
   properties {categoriesToShow [nocategory]}
   slot {class=StackBoxSlot}
  }
 ]
}
